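# Parent material for the FrioCrystal vehicle wheel cosmetic, giving rim and tire an animated, tintable crystal look.
# Rim layers combine rotating spirals and panning masks; tire layers add normal-distorted shimmer, sparkles and view-dependent highlights, with a Painted switch desaturating shimmer tints.
# Named reroutes carry these layers into lerped base color, emissive and normal, then through vehicle damage and de-rez transition functions.
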Material /VehicleCosmetics/Wheels/FrioCrystal/Materials/M_WheelParent_FrioCrystal
graph {
 properties {
  bAutomaticallySetUsageInEditor=false
  bUseMaterialAttributes=true
  bUsedWithMeshDeformer=true
  bUsedWithSkeletalMesh=true
  usedShadingModels=Default Lit
 }
 nodes [
  Constant() {r=1.2}
  n1 = SetMaterialAttributes!(
   MaterialAttributes
   Base Color=NamedRerouteUsage()
   Emissive Color=NamedRerouteUsage()
   Normal=NamedRerouteUsage()
   Metallic=ScalarParameter() {group=00 Base, parameterName=Metalic, sliderMax=1, sortPriority=10}
   Roughness=ScalarParameter() {defaultValue=1, group=00 Base, parameterName=Roughness, sliderMax=1, sortPriority=11}
  ) {attributeSetTypes [
    #1
    #2
    #3
    #4
    #5
   ]}
  MaterialFunctionCall(In=MaterialFunctionCall(Material=@n1) {materialFunction=&MF_BaseVehicle_Damage}.Result) {materialFunction=&MF_VehicleDeRez_Transition}
  MaterialFunctionCall!(Normal, Exponent=Constant() {r=1}) {materialFunction=/VehicleCosmetics/Wheels/Cola_Holographic/Materials/MF_Fresnel_Classic}
  n2 = MaterialFunctionCall!(Blend=NamedRerouteUsage(), Base=Constant3Vector() {constant=<rgba 0.177 0.177 0.177 0>}) {materialFunction=/Engine/Functions/Engine_MaterialFunctions03/Blends/Blend_Screen}
  n3 = NamedRerouteUsage()
  n4 = Panner() {speedX=0.1, speedY=0.01}
  n5 = NamedRerouteUsage()
  n6 = ConstantBiasScale(None=Multiply(
    A=Multiply(A=@n2.Result, B=@n3)
    B=TextureSample(UVs=@n4, Tex=@n5) {samplerSource=SSM_Wrap_WorldGroupSettings, texture=/VehicleCosmetics/Wheels/FrioCrystal/Textures/wheel_frio_SE_rgb_mask}.B
   )) {bias=0, scale=3}
  n7 = CameraVectorWS()
  Multiply(A=@n6, B=Power(Base=ComponentMask(None=@n7) {b=true}) {constExponent=1.5})
  n8 = NamedRerouteUsage()
  n9 = NamedRerouteUsage()
  n10 = NamedRerouteUsage()
  n11 = Add(A=NamedRerouteUsage(), B=@n10)
  n12 = NamedRerouteUsage()
  n13 = LinearInterpolate(A=NamedRerouteUsage(), B=Max(A=Add(A=@n9, B=@n11), B=@n12), Alpha=@n8)
  NamedRerouteDeclaration(None=LinearInterpolate(
    A=@n13
    B=ConstantBiasScale(None=@n13) {bias=0, scale=0.1}
    Alpha=ScalarParameter() {group=00 Base, parameterName=Base Color Darkening, sliderMax=1, sortPriority=12}
   )) {name=Base Color, nodeColor=<rgba 1 0.5859 0 1>}
  n14 = Max(A=@n9, B=@n10)
  NamedRerouteDeclaration(None=LinearInterpolate(A=NamedRerouteUsage(), B=Add(A=@n14, B=@n12), Alpha=@n8)) {name=Emissive, nodeColor=<rgba 0.2656 0 1 1>}
  n15 = TextureObjectParameter() {group=00 Base, parameterName=VFX, sortPriority=5, texture=&wheel_frio_SE_rgb_mask}
  NamedRerouteDeclaration(None=@n15) {name=VFX Tex, nodeColor=<rgba 0.6172 1 0 1>}
  NamedRerouteDeclaration(None=@n6) {name=Edges, nodeColor=<rgba 1 0 0.9375 1>}
  n16 = TextureSample() {
   samplerSource=SSM_Wrap_WorldGroupSettings
   samplerType=SAMPLERTYPE_Grayscale
   texture=/VehicleCosmetics/Wheels/_SharedTextures/FX/GradientCircle01
  }
  n17 = ConstantBiasScale(None=OneMinus(None=@n16.R)) {bias=-0.55, scale=1.5}
  n18 = Add(A=TextureCoordinate() {uTiling=0.8, vTiling=0.8}) {constB=0.1}
  n19 = Rotator(Coordinate=@n18) {speed=-0.5}
  n20 = Rotator(Coordinate=@n18) {speed=-0.4}
  n21 = Max(
   A=Clamp(None=TextureSample(UVs=@n19) {samplerSource=SSM_Wrap_WorldGroupSettings, texture=/VehicleCosmetics/Wheels/_SharedTextures/FX/alpha_spiral}.R)
   B=Clamp(None=TextureSample(UVs=@n20) {samplerSource=SSM_Wrap_WorldGroupSettings, texture=&alpha_spiral}.G)
  )
  n22 = Clamp(None=TextureSample(UVs=Rotator(Coordinate=@n18) {speed=0.15}) {samplerSource=SSM_Wrap_WorldGroupSettings, texture=&alpha_spiral}.B)
  n23 = Multiply(
   A=Multiply(A=Clamp(None=@n17), B=Max(A=@n21, B=@n22))
   B=ScalarParameter() {defaultValue=1, group=00 Base, parameterName=Spiral Intensity, sortPriority=13}
  )
  n24 = NamedRerouteUsage()
  n25 = NamedRerouteUsage()
  n26 = Constant3Vector() {constant=<rgba 0.889 0.889 0.889 0>}
  NamedRerouteDeclaration(None=Multiply(A=Multiply(A=@n23, B=@n24), B=Multiply(A=@n25, B=@n26))) {name=Spiral, nodeColor=<rgba 0.8984 0 1 1>}
  n27 = NamedRerouteUsage()
  n28 = NamedRerouteUsage()
  n29 = NamedRerouteUsage()
  n30 = Rotator() {speed=0.35}
  n31 = Multiply(
   A=Multiply(A=Constant3Vector() {constant=<rgba 0.442 0.442 0.442 0>}, B=@n29)
   B=TextureSample(UVs=Panner(Coordinate=@n30) {speedX=0.1, speedY=-0.1}, Tex=NamedRerouteUsage()) {samplerSource=SSM_Wrap_WorldGroupSettings, texture=&wheel_frio_SE_rgb_mask}.B
  )
  n32 = NamedRerouteUsage()
  n33 = MaterialFunctionCall(Base=Constant3Vector() {constant=<rgba 0.054 0.015 0.06 0>}, Blend=@n29) {materialFunction=&Blend_Overlay}
  n34 = TextureCoordinate() {vTiling=2}
  n35 = ConstantBiasScale(None=TextureSample(UVs=Panner(Coordinate=@n34) {speedX=0.05, speedY=-0.2}, Tex=NamedRerouteUsage()) {samplerSource=SSM_Wrap_WorldGroupSettings, texture=&wheel_frio_SE_rgb_mask}.G) {bias=0, scale=2}
  n36 = OneMinus(None=NamedRerouteUsage())
  n37 = Multiply(A=Add(A=Multiply(A=@n31, B=@n32), B=Multiply(A=@n33.Result, B=@n35)), B=Clamp(None=ConstantBiasScale(None=@n36) {bias=-0.5, scale=5}))
  n38 = OneMinus(None=CameraVectorWS())
  n39 = NamedRerouteUsage()
  n40 = Multiply(
   A=Power(Base=ComponentMask(None=@n38) {b=true}) {constExponent=1.2}
   B=Multiply(A=Constant3Vector() {constant=<rgba 0.136 0.675 0.878 0>}, B=@n39)
  )
  n41 = MaterialFunctionCall(Base=@n39, Blend=Constant3Vector() {constant=<rgba 0.41 0.08 0.165 0>}) {materialFunction=&Blend_Overlay}
  n42 = Multiply(A=Add(A=@n37, B=Add(A=@n40, B=@n41.Result)), B=Clamp(None=@n36) {minDefault=0.5})
  NamedRerouteDeclaration(None=Add(A=@n27, B=Add(A=@n28, B=@n42))) {name=Rim BC, nodeColor=<rgba 0 1 0.5781 1>}
  NamedRerouteDeclaration(None=Max(A=@n27, B=Add(A=@n28, B=@n37))) {name=Rim Emissive, nodeColor=<rgba 0 1 0.5781 1>}
  n43 = Time()
  n44 = Add(A=Sine(None=Multiply(A=@n43)) {period=3})
  NamedRerouteDeclaration(None=ConstantBiasScale(None=Multiply(A=@n44) {constB=0.25}) {bias=0.4, scale=1}) {name=Time_1, nodeColor=<rgba 1 0 0.8672 1>}
  n45 = Add(A=Sine(None=Multiply(A=@n43) {constB=0.4}) {period=3})
  NamedRerouteDeclaration(None=ConstantBiasScale(None=Multiply(A=@n45) {constB=0.25}) {bias=0.1, scale=1.5}) {name=Time_2, nodeColor=<rgba 1 0 0.8672 1>}
  n46 = TextureSampleParameter2D() {
   group=00 Base
   parameterName=Tire Normal
   samplerSource=SSM_Wrap_WorldGroupSettings
   samplerType=SAMPLERTYPE_Normal
   sortPriority=4
   texture=/VehicleCosmetics/Wheels/FrioCrystal/Textures/T_Wheel_Frio_Crystal_Tire_N
  }
  NamedRerouteDeclaration(None=@n46.RGB) {name=Tire Normal Map, nodeColor=<rgba 0.5 0.5 1 1>}
  NamedRerouteDeclaration(None=VertexColor().R) {name=VC_R, nodeColor=<rgba 0 1 0.0859 1>}
  n47 = Clamp(None=NamedRerouteUsage())
  n48 = Clamp(None=@n47) {minDefault=0.25}
  n49 = ConstantBiasScale(None=ComponentMask(None=NamedRerouteUsage()) {g=true, r=true}) {bias=-0.5, scale=2}
  n50 = Add(
   A=Multiply(A=Clamp(None=@n49), B=ScalarParameter() {defaultValue=0.3, group=00 Base, parameterName=Shimmer 1 Normal Distort, sortPriority=14})
   B=TextureCoordinate()
  )
  n51 = NamedRerouteUsage()
  n52 = TextureSample(UVs=Panner(Coordinate=@n50) {speedX=0.1, speedY=0.02}, Tex=@n51) {samplerSource=SSM_Wrap_WorldGroupSettings, samplerType=SAMPLERTYPE_Normal, texture=/VehicleCosmetics/Wheels/FrioCrystal/Textures/tire_frio_n}
  n53 = Multiply(A=@n52.G, B=Multiply(A=NamedRerouteUsage(), B=Constant3Vector() {constant=<rgba 0.317 0.167 0.343 0>}))
  n54 = TextureSample(UVs=Panner(Coordinate=@n50) {speedX=0.3, speedY=0.04}, Tex=@n51) {samplerSource=SSM_Wrap_WorldGroupSettings, samplerType=SAMPLERTYPE_Normal, texture=&tire_frio_n}
  n55 = Constant3Vector() {constant=<rgba 0.317 0.167 0.343 0>}
  n56 = StaticSwitchParameter(True=Desaturation(None=@n55) {luminanceFactors=<rgba 0.3 0.59 0.11 0>}, False=@n55) {defaultValue=true, group=00 Base, parameterName=Painted?, sortPriority=1}
  n57 = NamedRerouteUsage()
  n58 = Multiply(
   A=@n47
   B=Multiply(
    A=ConstantBiasScale(None=@n54.R) {bias=0, scale=1}
    B=MaterialFunctionCall(Base=@n56, Blend=@n57) {materialFunction=&Blend_Overlay}.Result
   )
  )
  n59 = NamedRerouteUsage()
  n60 = TextureSample(UVs=Panner(Coordinate=@n50) {speedX=0.3, speedY=-0.04}, Tex=@n51) {samplerSource=SSM_Wrap_WorldGroupSettings, samplerType=SAMPLERTYPE_Normal, texture=&tire_frio_n}
  n61 = Constant3Vector() {constant=<rgba 0.436 0.118 0.12 0>}
  n62 = Desaturation(None=@n61) {luminanceFactors=<rgba 0.3 0.59 0.11 0>}
  n63 = Multiply(
   A=ConstantBiasScale(None=@n60.G) {bias=-0.05, scale=1}
   B=MaterialFunctionCall(Base=StaticSwitchParameter(True=@n62, False=@n61) {defaultValue=true, group=00 Base, parameterName=Painted?, sortPriority=1}, Blend=@n57) {materialFunction=&Blend_Overlay}.Result
  )
  n64 = Add(A=Add(A=Multiply(A=@n48, B=@n53), B=Multiply(A=@n58, B=@n59)), B=Multiply(A=Multiply(A=@n47, B=@n63), B=@n59))
  n65 = TextureSample(UVs=Panner(Coordinate=@n50) {speedX=0.2, speedY=0.02}, Tex=@n51) {samplerSource=SSM_Wrap_WorldGroupSettings, samplerType=SAMPLERTYPE_Normal, texture=&tire_frio_n}
  n66 = Constant3Vector() {constant=<rgba 0.608 0.032 0.331 0>}
  n67 = Desaturation(None=@n66) {luminanceFactors=<rgba 0.3 0.59 0.11 0>}
  n68 = Multiply(
   A=ConstantBiasScale(None=@n65.B) {bias=-0.2, scale=1.5}
   B=MaterialFunctionCall(Base=StaticSwitchParameter(True=@n67, False=@n66) {defaultValue=true, group=00 Base, parameterName=Painted?, sortPriority=1}, Blend=@n57) {materialFunction=&Blend_Overlay}.Result
  )
  n69 = Multiply(A=Multiply(A=@n47, B=ConstantBiasScale(None=@n68) {bias=0, scale=1.5}), B=NamedRerouteUsage())
  NamedRerouteDeclaration(None=Clamp(None=Add(A=@n64, B=@n69))) {name=Tire Colors, nodeColor=<rgba 0.1484 1 0 1>}
  NamedRerouteDeclaration(None=LinearInterpolate(A=NamedRerouteUsage(), B=NamedRerouteUsage(), Alpha=NamedRerouteUsage())) {name=Normal, nodeColor=<rgba 0.5 0.5 1 1>}
  n70 = NamedRerouteUsage()
  n71 = Constant3Vector() {constant=<rgba 0.608 0.608 0.608 0>}
  n72 = TextureSample(UVs=Panner() {speedX=0.1, speedY=0.01}, Tex=NamedRerouteUsage()) {samplerSource=SSM_Wrap_WorldGroupSettings, samplerType=SAMPLERTYPE_Normal, texture=&tire_frio_n}
  n73 = Multiply(
   A=Multiply(
    A=MaterialFunctionCall(Base=@n70, Blend=@n71) {materialFunction=&Blend_Overlay}.Result
    B=NamedRerouteUsage()
   )
   B=@n72.B
  )
  n74 = ComponentMask(None=OneMinus(None=CameraVectorWS())) {b=true}
  NamedRerouteDeclaration(None=Multiply(A=ConstantBiasScale(None=@n73) {bias=0, scale=5}, B=Power(Base=@n74))) {name=Tire Curvature, nodeColor=<rgba 0 1 0.7188 1>}
  n75 = Multiply(A=NamedRerouteUsage(), B=Constant3Vector() {constant=<rgba 0.298 0.298 0.298 0>})
  n76 = OneMinus(None=CameraVectorWS())
  n77 = Multiply(A=@n75, B=Power(Base=ComponentMask(None=@n76) {b=true}) {constExponent=1})
  NamedRerouteDeclaration(None=Add(A=ConstantBiasScale(None=@n77) {bias=0, scale=3}, B=@n75)) {name=Tire Fresnel, nodeColor=<rgba 1 0.3984 0 1>}
  NamedRerouteDeclaration(None=TextureObjectParameter() {group=00 Base, parameterName=Sparkles, sortPriority=6, texture=/VehicleCosmetics/Wheels/_SharedTextures/FX/ice_rgb}) {name=Sparkles Tex, nodeColor=<rgba 0.6172 1 0 1>}
  n78 = Multiply(
   A=BumpOffset(Coordinate=TextureCoordinate(), Height=Constant())
   B=ScalarParameter() {defaultValue=5, group=00 Base, parameterName=Sparkle Tiling, sortPriority=15}
  )
  n79 = NamedRerouteUsage()
  n80 = TextureSample(UVs=Panner(Coordinate=@n78) {speedY=0.3}, Tex=@n79) {samplerSource=SSM_Wrap_WorldGroupSettings, samplerType=SAMPLERTYPE_Normal, texture=&tire_frio_n}
  n81 = TextureSample(UVs=Panner(Coordinate=@n78) {speedY=0.05}, Tex=@n79) {samplerSource=SSM_Wrap_WorldGroupSettings, samplerType=SAMPLERTYPE_Normal, texture=&tire_frio_n}
  n82 = Clamp(None=ConstantBiasScale(None=Multiply(A=@n80.G, B=@n81.R)) {bias=-0.15, scale=100}) {maxDefault=2}
  n83 = MaterialFunctionCall(Exponent=Constant() {r=2}) {materialFunction=&MF_Fresnel_Classic}
  NamedRerouteDeclaration(None=Multiply(A=Multiply(A=@n82, B=@n83.Result), B=NamedRerouteUsage())) {name=Sparkles, nodeColor=<rgba 0.3359 1 0 1>}
  NamedRerouteDeclaration(None=VectorParameter() {group=00 Base, parameterName=Custom Color, sortPriority=0, defaultValue=<rgba 0.0782 0.7913 1 1>}.RGB) {name=Custom Color, nodeColor=<rgba 1 0 0.375 1>}
  n84 = TextureSampleParameter2D() {
   group=00 Base
   parameterName=Rim Normal
   samplerType=SAMPLERTYPE_Normal
   sortPriority=3
   texture=/VehicleCosmetics/Wheels/FrioCrystal/Textures/T_Wheel_Frio_Crystal_Rim_N
  }
  NamedRerouteDeclaration(None=@n84.RGB) {name=Rim Normal Map, nodeColor=<rgba 0.5 0.5 1 1>}
  n85 = TextureSampleParameter2D() {
   group=00 Base
   parameterName=Tire Masks
   samplerType=SAMPLERTYPE_LinearColor
   sortPriority=2
   texture=/VehicleCosmetics/Wheels/FrioCrystal/Textures/T_Wheel_Frio_Crystal_Tire_RGB
  }
  NamedRerouteDeclaration(None=@n85.B) {name=Tire Masks B, nodeColor=<rgba 0 0 1 1>}
  NamedRerouteDeclaration(None=@n85.G) {name=Tire Masks G, nodeColor=<rgba 0 1 0 1>}
  n86 = TextureSampleParameter2D() {
   group=00 Base
   parameterName=Rim Masks
   samplerType=SAMPLERTYPE_LinearColor
   sortPriority=1
   texture=/VehicleCosmetics/Wheels/FrioCrystal/Textures/T_Wheel_Frio_Crystal_Rim_RGB
  }
  NamedRerouteDeclaration(None=@n86.A) {name=Rim Masks A, nodeColor=<rgba 1 1 1 1>}
  NamedRerouteDeclaration(None=@n86.G) {name=Rim Masks G, nodeColor=<rgba 0 1 0 1>}
  NamedRerouteDeclaration(None=@n86.R) {name=Rim Masks R, nodeColor=<rgba 0.0781 1 0 1>}
  OneMinus(None=MaterialFunctionCall(Exponent=Constant() {r=1.5}) {materialFunction=&MF_Fresnel_Classic}.Result)
  OneMinus(None=MaterialFunctionCall(Exponent=Constant() {r=1}) {materialFunction=&MF_Fresnel_Classic}.Result)
 ]
}
